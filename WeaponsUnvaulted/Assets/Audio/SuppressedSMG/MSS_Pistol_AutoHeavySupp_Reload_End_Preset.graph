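# MetaSound Source preset for the reload-end sound of the suppressed auto heavy pistol in the SuppressedSMG weapon audio set.
# It inherits its graph from MSS_Weapon_Reload_End, supplying three wave variations along with master volume and random pitch-shift inputs.
# Output is a mono one-shot routed through the Fortnite weapon foley sound class, attenuation and concurrency settings.
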MetaSoundSource /WeaponsUnvaulted/Assets/Audio/SuppressedSMG/MSS_Pistol_AutoHeavySupp_Reload_End_Preset
preset_parent=/FortMetasound/Weapons/MSS_Weapon_Reload_End
inputs [
 MasterVolume {type=Float}
 Random.PitchShift.Base {type=Float}
 Random.PitchShift.Spread {type=Float}
 Sounds {type=WaveAsset:Array}
 UE.Source.OnPlay {type=Trigger}
]
interfaces [UE.OutputFormat.Mono:1.0, UE.Source.OneShot:1.0, UE.Source:1.1]
outputs [
 "UE.OutputFormat.Mono.Audio:0" {type=Audio}
 UE.Source.OneShot.OnFinished {type=Trigger}
]
pages [Default {input_defaults {
   MasterVolume=0.5
   Random.PitchShift.Base=0
   Random.PitchShift.Spread=0.5
   UE.Source.OnPlay=null
   Sounds [
    /WeaponsUnvaulted/Assets/Audio/SuppressedSMG/SoundWaves/Pistol_AutoHeavySupp_Reload_End_01
    /WeaponsUnvaulted/Assets/Audio/SuppressedSMG/SoundWaves/Pistol_AutoHeavySupp_Reload_End_02
    /WeaponsUnvaulted/Assets/Audio/SuppressedSMG/SoundWaves/Pistol_AutoHeavySupp_Reload_End_03
   ]
  }}]
properties {
 assetImportData=&self:AssetImportData
 attenuationSettings=&AS_Fort_WeaponFoley
 maxDistance=2250
 sampleRate=48000
 soundClassObject=&Fort_WeaponFoley
 volume=1.1
 concurrencySet [&CS_Fort_WeaponFoley, null]
}
